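; Blueprint for a brick chimney fireplace prop in the Figment environment, with fire sound and point light components.
; Its construction script calls the parent's, then checks a Turn Off flag.
; When the flag is set, it fades out the fire audio over half a second and destroys the point light, leaving the fireplace unlit and silent.
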(fn ConstructionScript
  (|CallParentFunction)
  (if (Variables|Default|GetTurn%20Off)
    (Audio|Components|Audio|FadeOut (Variables|Default|GetFire_Sound) 0.5 0.0 "Linear" :!class /Script/Engine.AudioComponent)
    (Components|DestroyComponent (Variables|Default|GetPointLight))))
