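; Figment_Durrr_Sign_01 is a sign actor Blueprint that destroys itself when the actor it is linked to is destroyed.
; On the FortLinkToActor OnLinkedActorDestroyed event, it applies a very large environmental damage hit to itself, passing along the damage causer.
; The construction script is empty.
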(event OnLinkedActorDestroyed(FortLinkToActor)
  (params (DamageCauser :type Actor%20Object%20Reference))
  (Game|Damage|ApplyDamage self 1e+04 0 DamageCauser "/Engine/EngineDamageTypes/DmgTypeBP_Environmental.DmgTypeBP_Environmental_C"))

(fn ConstructionScript)
