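# Default-lit material for the Winterfest stocking prop in the Figment environment, with diffuse, specular and normal textures, Fresnel brightness terms and procedural vertex-offset animation.
# Vertex color masks drive the deformation: R collapses vertices to the object position, while B outside R squeezes along local X with a power falloff.
# Rounded Progress switches that deformation off at one.
# ProgressJiggle scales an added sine-driven rotation about local X around a pivot parameter.
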Material /FigmentEnvironment/Materials/M_WinterfestStocking_Figment
graph {
 properties {pixelDepthOffsetMode=PDOM_Legacy, usedShadingModels=Default Lit}
 nodes [
  Add(A=OneMinus(None=VertexColor().R))
  n1 = Fresnel(ExponentIn=ScalarParameter() {defaultValue=3, parameterName=Fresnel}, Normal=PixelNormalWS()) {baseReflectFraction=0}
  n2 = ScalarParameter() {defaultValue=5, parameterName=FresnelBrightness2}
  Add(A=Multiply(A=ScalarParameter() {defaultValue=1, parameterName=FresnelBrightness}, B=@n1), B=Multiply(A=@n1, B=Multiply(A=@n2)))
  n3 = VertexColor()
  n4 = Subtract(A=ObjectPositionWS(), B=WorldPosition().XYZ)
  n5 = Multiply(A=OneMinus(None=@n3.R), B=@n3.B)
  n6 = WorldPosition()
  n7 = ObjectPositionWS()
  n8 = ComponentMask(None=Transform(None=Subtract(A=@n6.XYZ, B=@n7)) {transformSourceType=TRANSFORMSOURCE_World, transformType=TRANSFORM_Local}) {r=true}
  n9 = Multiply(A=Constant3Vector() {constant=<rgba 1 0 0 0>}, B=Sign(None=@n8))
  n10 = Multiply(
   A=Multiply(A=Transform(None=@n9) {transformSourceType=TRANSFORMSOURCE_Local}, B=ScalarParameter() {defaultValue=-0.4327, parameterName=Squeeze})
   B=Power(Base=Abs(None=@n8), Exp=ScalarParameter() {defaultValue=1.2468, parameterName=SqueezeFalloff})
  )
  n11 = ScalarParameter() {defaultValue=1, parameterName=Progress}
  n12 = LinearInterpolate(A=Add(A=Multiply(A=@n3.R, B=@n4), B=Multiply(A=@n5, B=@n10)), B=Constant3Vector(), Alpha=Round(None=Saturate(None=@n11)))
  n13 = Normalize(VectorInput=Transform(None=Constant3Vector() {constant=<rgba 1 0 0 0>}) {transformSourceType=TRANSFORMSOURCE_Local})
  n14 = Time()
  n15 = ScalarParameter() {defaultValue=2, parameterName=SpeedRot}
  n16 = Multiply(A=Sine(None=Multiply(A=@n14, B=@n15)), B=ScalarParameter() {defaultValue=0.02, parameterName=RotJiggle})
  n17 = TransformPosition(None=VectorParameter() {parameterName=pivot, defaultValue=<rgba 0 13.7973 -9.4825 1>}.RGB) {transformType=TRANSFORMPOSSOURCE_World}
  n18 = Add(A=@n12, B=WorldPosition().XYZ)
  Add(
   A=@n12
   B=Multiply(
    A=ScalarParameter() {parameterName=ProgressJiggle}
    B=RotateAboutAxis(NormalizedRotationAxis=@n13, RotationAngle=@n16, PivotPoint=@n17, Position=@n18)
   )
  )
  Multiply()
  TextureSample() {texture=/Game/Environments/Apollo/Props/Winterfest/Textures/T_WinterStocking1_D}
  TextureSample() {samplerType=SAMPLERTYPE_Masks, texture=/Game/Environments/Apollo/Props/Winterfest/Textures/T_WinterStocking1_S}
  TextureSample() {samplerType=SAMPLERTYPE_Normal, texture=/Game/Environments/Apollo/Props/Winterfest/Textures/T_WinterStocking1_N}
 ]
}
